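; Velvet Thunder slushie machine that players interact with to make a Super Squishee, ejected as tossed loot pickups.
; SpawnItem blocks interaction and runs a timed charge-up, then GiveTreat spawns loot on authority with nozzle FX and re-enables interaction later.
; Replicated ToggleDrip drives a cosmetic timer that raises shake and drains fill through Custom Primitive Data, flagging out-of-stock when low.
; Conversation events track the player so only their client hears the menu sound.
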(event EventBeginPlay
  (CallFunction|SetFillLevelVisuals (Variables|Default|GetSlushAmount)))

(event Custom|GiveTreat
  (bind (_returnvalue_x _returnvalue_y _returnvalue_z) (Math|Vector|BreakVector (Transformation|GetForwardVector (Variables|Default|GetThrowPoint))))
  (Utilities|FlowControl|Sequence
    (:then_0
      (Rendering|Material|SetCustomPrimitiveDataFloat (Variables|Debug|GetStaticMeshComponent) 4 0.0)
      (Audio|PlaySoundatLocation "/WildEstateConsumables/Gameplay/VelvetThunder/Assets/Audio/VelvetThunder_Activate_End_Cue.VelvetThunder_Activate_End_Cue" (Transformation|GetActorLocation))
      (Networking|SwitchHasAuthority
        (:Authority
          (bind _itempickupsspawned (Class|AthenaFunctionLibrary|SpawnLootfromTableUsingItemEntry (Variables|Default|GetLootTierToSpawn) 0 -1 (+ (Transformation|GetWorldLocation (Variables|Default|GetThrowPoint)) (* (Transformation|GetForwardVector (Variables|Default|GetThrowPoint)) 5e+01)) (Transformation|GetForwardVector (Variables|Default|GetThrowPoint)) 3 true true false 0 "Container" "Chest"))
          (for (_array_element _array_index) _itempickupsspawned
            ;[0.3166 0.4028 1] Use a more authored toss because the standard one is unreliable
            (Fortnite|Container|TossPickupfromContainer self _array_element (- (Utilities|Array|Length _itempickupsspawned) 1) _array_index 9e+01 (Math|Rotator|MakeRotator _returnvalue_x _returnvalue_y _returnvalue_z) (Math|Random|RandomFloatinRange 6e+02 8e+02) false false 0))
          (Niagara|SpawnSystematLocation :SystemTemplate "/WildEstateTech/FX/VelvetThunder/NS_VelvetThunder_Eject_Single.NS_VelvetThunder_Eject_Single" :Location (Math|Vector|SelectVector (Transformation|GetWorldLocation (Variables|Default|GetLeftNozzle)) (Transformation|GetWorldLocation (Variables|Default|GetRightNozzle)) (> (Variables|Default|GetSlushAmount) 0.5)) :Scale "1.000000,1.000000,1.000000" :bAutoDestroy false :bAutoActivate true :PoolingMethod "None" :bPreCullCheck true))))
    (:then_1
      (Utilities|Time|SetTimerbyEvent (delegate Custom|AllowInteract) 2.0))))

(event Custom|AllowInteract
  (Variables|Interaction|SetbAllowInteract true))

(event Custom|SpawnItem
  ;[1 1 1] Trigger spawn effects and block interact
  (Variables|Interaction|SetbAllowInteract false)
  (bind _output_get (Variables|Default|Setinitialworldlocation (Transformation|GetWorldLocation (Variables|Debug|GetStaticMeshComponent))))
  (bind _returnvalue (Audio|SpawnSoundatLocation (Variables|Default|GetChargeUpSound) _output_get))
  (Variables|Default|SetActiveAudioLoop _returnvalue)
  (SoundIndicator|StartSoundIndicatorLooping (Variables|Default|GetFortSoundIndicator) (Transformation|GetActorLocation self) (Variables|Default|GetChargeUpSound) 1.0 0)
  (Variables|Default|SetToggleDrip false)
  (bind _returnvalue_2 (Utilities|Time|SetTimerbyEvent (EventDispatchers|CreateEvent :!event Custom|FinishSpawnShake) (Variables|Default|GetTimeToSpawn)))
  (Variables|Default|SetSpawnTimerHandle _returnvalue_2)
  (Development|PrintString "Start Saking the machine!" true true "(R=0.000000,G=0.660000,B=1.000000,A=1.000000)" 5e+01))

(event Custom|FinishSpawnShake
  ;[1 1 1] Trigger spawn effects and block interact
  (Utilities|Time|ClearandInvalidateTimerbyHandle (Variables|Default|GetSpawnTimerHandle))
  (Audio|Components|Audio|Stop (Variables|Default|GetActiveAudioLoop))
  (CallFunction|StopShakeEffects)
  (Variables|Default|SetToggleDrip true)
  (CallFunction|GiveTreat)
  (SoundIndicator|StopSoundIndicatorLooping (Variables|Default|GetFortSoundIndicator)))

(event Custom|StopShakeEffects
  (Audio|Components|Audio|Stop (Variables|Default|GetActiveAudioLoop))
  (Rendering|Material|SetCustomPrimitiveDataFloat (Variables|Debug|GetStaticMeshComponent) 2 0.0))

(event OnSingleConversationEnded(NonPlayerConversationComponent)
  (params (ParticipantActorPlayer :type Actor%20Object%20Reference))
  ;[1 1 1] Get Player in conversation and play sound for conversation start.
  (Variables|Default|SetConversingPlayer 0))

(event OnConversationEnded(NonPlayerConversationComponent)
  ;[1 1 1] Get Player in conversation and play sound for conversation start.
  (Variables|Default|SetConversingPlayer 0))

(event OnEnterConversationStateDelegate(NonPlayerConversationComponent)
  ;[1 1 1] Get Player in conversation and play sound for conversation start.
  (for _array_element (Conversation|GetOtherParticipantActors (Variables|Components|GetNonPlayerConversationComponent))
    (bind _asfort_player_controller (Utilities|Casting|CastToFortPlayerController _array_element))
    (Variables|Default|SetConversingPlayer _asfort_player_controller)))

(event Custom|DrainLiquidTimer
  (bind _output_get (Variables|Default|SetShakeIntensity (Math|Float|Clamp(Float) (+ (Variables|Default|GetShakeIntensity) 0.1))))
  (CallFunction|SetShakeIntensityVisuals _output_get)
  (Variables|Default|SetSlushAmount (Math|Float|Clamp(Float) (- (Variables|Default|GetSlushAmount) (Variables|Default|GetDrainPerLoop))))
  (CallFunction|SetFillLevelVisuals (Variables|Default|GetSlushAmount)))

(fn BlueprintCanInteract
  (params (InteractingPawn :type Fort%20Pawn%20Object%20Reference) (InteractionBeingAttempted :type EInteractionBeingAttempted%20Enum :default "FirstInteraction") (InteractionType :type TInteractionType%20Enum :default "IT_NoInteraction"))
  (return (== (|CallParentFunction InteractingPawn InteractionBeingAttempted InteractionType) (Variables|Interaction|GetbAllowInteract))))

(fn BlueprintGetInteractionString
  (params (InteractingPawn :type Fort%20Pawn%20Object%20Reference) (InteractionBeingAttempted :type EInteractionBeingAttempted%20Enum :default "FirstInteraction"))
  (return "NSLOCTEXT(\"[9928BFAEB0E7D60589F8C699D832724E]\", \"5AEAAB344D74CD550E873BBD572E6B14\", \"Make Super Squishee\")"))

(fn BlueprintOnLocalInteract
  (params (InteractingPawn :type Fort%20Player%20Pawn%20Object%20Reference) (InteractionBeingAttempted :type EInteractionBeingAttempted%20Enum :default "FirstInteraction"))
  (bind _returnvalue (|CallParentFunction InteractingPawn InteractionBeingAttempted))
  (return (== _returnvalue (Variables|Interaction|GetbAllowInteract))))

(fn OnRep_ConversingPlayer
  ;[0 0.15 0.07284 0.5] Confirm player is locally controlled on client side, then play audio
  (Utilities|IsValid (Variables|Default|GetConversingPlayer)
    (:Is%20Valid
      (Networking|CanExecuteCosmeticEvents
        (:True
          (if (Utilities|Array|ContainsItem (Fortnite|GetLocalFortPlayerControllers true) (Variables|Default|GetConversingPlayer))
            (Audio|PlaySoundatLocation "/WildEstateConsumables/Gameplay/VelvetThunder/Assets/Audio/VelvetThunder_Menu_Open_Cue.VelvetThunder_Menu_Open_Cue" (Transformation|GetActorLocation))))))))

(fn OnRep_ToggleDrip
  (bind _toggledrip (Variables|Default|GetToggleDrip))
  (Networking|CanExecuteCosmeticEvents
    (:True
      (Niagara|SetNiagaraVariable(Int32) (Variables|Default|GetIdleFX) "SpawnDrips" (select (and (not (Variables|Default|GetOutOfStock)) _toggledrip) 1 0))
      (if _toggledrip
        (Utilities|Time|ClearandInvalidateTimerbyHandle (Variables|Default|GetDrainTimerHandle))
        (else
          (bind _returnvalue (Utilities|Time|SetTimerbyEvent (EventDispatchers|CreateEvent :!event DrainLiquidTimer) 0.2 true))
          (Variables|Default|SetDrainTimerHandle _returnvalue))))))

(fn SetFillLevelVisuals
  (params (Alpha :type Float%20%28single-precision%29))
  (Networking|CanExecuteCosmeticEvents
    (:True
      (Rendering|Material|SetCustomPrimitiveDataFloat (Variables|Debug|GetStaticMeshComponent) 1 Alpha)
      (if (<= Alpha 0.1)
        (Niagara|SetNiagaraVariable(Int32) (Variables|Default|GetIdleFX) "SpawnDrips" 0)
        (Variables|Default|SetOutOfStock true)))))

(fn SetShakeIntensityVisuals
  (params (Alpha :type Float%20%28single-precision%29))
  (Rendering|Material|SetCustomPrimitiveDataFloat (Variables|Default|GetSM_FrozenFood_SlushieLargeCupGlass_A) 2 Alpha))

(fn ConstructionScript)
