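# Conversation database for VelvetThunder, a talking, armless robot NPC that chats affectionately with players and offers a sale.
# The entry point opens a Speech node whose styled lines are gated by low player health, Victory Crown ownership or, through inverted pitch/roll checks, the NPC being tipped over.
# Crown lines carry raised priority.
# Speech leads to a Sell node requiring the Sell1 service, which skips normal item granting and carries a SpawnItem sub-node.
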ConversationDatabase /WildEstateConsumables/Gameplay/VelvetThunder/Conversation_WildEstate
entry_points [{tag=Conversation.Entry.VelvetThunder, destinations [#1]}]
nodes [
 ConversationEntryPointNode_0 {
  id=#1
  class=ConversationEntryPointNode
  outputs [#2]
  properties {entryTag {tagName=Conversation.Entry.VelvetThunder}}
 }
 Speech_C_0 {
  id=#2
  class=Speech_C
  outputs [#3]
  properties {
   bUseInlineConfig=false
   choiceSound=/WildEstateConsumables/Gameplay/VelvetThunder/Assets/Audio/VelvetThunder_Menu_Open_Cue
   nodeName=""
   generalConfig {contextualMessages [
     {
      message="You're all banged up, bud!! Want a Patch Up?"
      contextRequirements [{requirement {
         class=FortControllerRequirement_AttributeThresholdCheck
         properties {
          queryAttribute {attribute=/Script/FortniteGame.FortHealthSet:Health, attributeName=Health, attributeOwner=&FortHealthSet}
          thresholdTestConfig {requirement=LessThanOrEqual, quantity {value=50}}
         }
        }}]
      metadataParameters [Style {value=Sad}]
      priority {curve {curveTable=null, rowName=Tandem.Speech.Priority.WhileSincePlayed}}
     }
     {
      message="Heeeey champ! You gotta take care of yourself! How 'bout some healing?"
      contextRequirements []
      metadataParameters [Style {value=Laugh}]
      priority {value=0, curve {curveTable=null, rowName=None}}
     }
     {
      message="Bestie!! ...Is it okay to call you bestie? or is that, like, so cringe?"
      contextRequirements []
      metadataParameters [Style {value=Excited}]
      priority {value=0, curve {curveTable=null, rowName=None}}
     }
     {
      message="Golly, I wish I had arms to hug ya!"
      contextRequirements []
      metadataParameters [Style {value=Blush}]
      priority {value=0, curve {curveTable=null, rowName=None}}
     }
     {
      message="All your bones are still in the right place! Yay!"
      contextRequirements []
      metadataParameters [Style {value=Excited}]
      priority {value=0, curve {curveTable=null, rowName=None}}
     }
     {
      message=How's the battle going? Need a pep talk?
      contextRequirements []
      metadataParameters [Style {value=Happy}]
      priority {value=0, curve {curveTable=null, rowName=None}}
     }
     {
      message="My circuits are figuratively bursting with love and respect for you!"
      contextRequirements []
      metadataParameters [Style {value=Sad}]
      priority {value=0, curve {curveTable=null, rowName=None}}
     }
     {
      message=You broke my soul... and my heart.
      contextRequirements [{
        participantId {tagName=Conversation.Participant.NPC}
        requirement {
         class=FortControllerRequirement_CheckPitchYawRoll
         properties {
          bInverseCondition=true
          pitchErrorCheck {bEnabled=true, toleratedErrorInDegrees=45}
          rollErrorCheck {bEnabled=true, toleratedErrorInDegrees=45}
         }
        }
       }]
      metadataParameters [Style {value=Sad}]
      priority {curve {curveTable=null, rowName=None}}
     }
     {
      message=...I guess I'll make a some little bug friends down here...
      contextRequirements [{
        participantId {tagName=Conversation.Participant.NPC}
        requirement {
         class=FortControllerRequirement_CheckPitchYawRoll
         properties {
          bInverseCondition=true
          pitchErrorCheck {bEnabled=true, toleratedErrorInDegrees=45}
          rollErrorCheck {bEnabled=true, toleratedErrorInDegrees=45}
         }
        }
       }]
      metadataParameters [Style {value=Blush}]
      priority {curve {curveTable=null, rowName=None}}
     }
     {
      message=The ground makes my circuits taste like sadness.
      contextRequirements [{
        participantId {tagName=Conversation.Participant.NPC}
        requirement {
         class=FortControllerRequirement_CheckPitchYawRoll
         properties {
          bInverseCondition=true
          pitchErrorCheck {bEnabled=true, toleratedErrorInDegrees=45}
          rollErrorCheck {bEnabled=true, toleratedErrorInDegrees=45}
         }
        }
       }]
      metadataParameters [Style {value=Sad}]
      priority {curve {curveTable=null, rowName=None}}
     }
     {
      message=B-best friend... how could you...
      contextRequirements [{
        participantId {tagName=Conversation.Participant.NPC}
        requirement {
         class=FortControllerRequirement_CheckPitchYawRoll
         properties {
          bInverseCondition=true
          pitchErrorCheck {bEnabled=true, toleratedErrorInDegrees=45}
          rollErrorCheck {bEnabled=true, toleratedErrorInDegrees=45}
         }
        }
       }]
      metadataParameters [Style {value=Sad}]
      priority {curve {curveTable=null, rowName=None}}
     }
     {
      message="A CROWN?! I'm so proud of you!"
      contextRequirements [{requirement {class=FortControllerRequirement_HasOwnedTag, properties {ownedTag {tagName=Granted.Abilities.VictoryCrown.CrownWearer}}}}]
      metadataParameters [Style {value=Excited}]
      priority {value=2, curve {curveTable=null, rowName=None}}
     }
     {
      message="I was making you a best friend crown, but that one's MUCH shinier..."
      contextRequirements [{requirement {class=FortControllerRequirement_HasOwnedTag, properties {ownedTag {tagName=Granted.Abilities.VictoryCrown.CrownWearer}}}}]
      metadataParameters [Style {value=Blush}]
      priority {value=2, curve {curveTable=null, rowName=None}}
     }
    ]}
  }
 }
 VelvetThunder_Sell_C_0 {
  id=#3
  class=VelvetThunder_Sell_C
  outputs []
  properties {
   bSkipGrantingItem=true
   choiceSound=/WildEstateConsumables/Gameplay/VelvetThunder/Assets/Audio/VelvetThunder_Menu_Purchase_Cue
   softTaskIcon=/WildEstateConsumables/Gameplay/SteelKite/Icons/T_Icon_WildEstate_SteelKite
  }
  sub_nodes [
   FortConversationRequirement_HasService_1 {class=FortConversationRequirement_HasService, properties {serviceTag {tagName=Tandem.Service.Sell1}}}
   VelvetThunder_SpawnItem_C_0 {class=VelvetThunder_SpawnItem_C}
  ]
 }
]
