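; A conversation side effect that runs on the server to dispense an item from a Velvet Thunder machine in WildEstate.
; It casts the conversation's NPC participant to BP_WildEstate_VelvetThunder and calls SpawnItem on it.
; It then casts the player participant to FortPlayerController and sends it the Event.WildEstate.VelvetThunder.Buy gameplay event.
; Either failed cast writes a shipping log message; an invalid player is only detected after the item spawns.
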(fn ServerCauseSideEffect
  (params (Context :type Conversation%20Context%20Structure%20%28by%20ref%29))
  (bind _returnvalue (Class|FortConversationContextLibrary|GetParticipantActorNPC Context))
  (bind _asbp_wild_estate_velvet_thunder (Utilities|Casting|CastToBP_WildEstate_VelvetThunder _returnvalue)
    (:then
      (Class|BPWildEstateVelvetThunder|SpawnItem _asbp_wild_estate_velvet_thunder)
      (bind _returnvalue_2 (Class|FortConversationContextLibrary|GetParticipantActorPlayer Context))
      (bind _asfort_player_controller (Utilities|Casting|CastToFortPlayerController _returnvalue_2)
        (:then
          (Ability|SendGameplayEventtoActor _asfort_player_controller "(TagName=\"Event.WildEstate.VelvetThunder.Buy\")")
          (return))
        (:CastFailed
          (Fortnite|FortShippingLog self "VelvetThunder_SpawnItem: Side effect Velvet Thunder player is Invalid!")
          (return))))
    (:CastFailed
      (Fortnite|FortShippingLog self "VelvetThunder_SpawnItem: Side effect Velvet Thunder machine is Invalid!" true)
      (return))))
